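# Material function in the Misc/Fortnite library, meant to adapt World Aligned Texture Complex output to Fortnite's spec texture format.
# It breaks an incoming material attributes input, which defaults to an empty preview set, and rebuilds it for output.
# As wired here, each connected output attribute reads the BaseColor pin, with no visible channel swizzle.
# Anisotropy, Tangent, ShadingModel and Displacement are left unconnected.
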MaterialFunction /FigmentEnvironment/FN_6_20/Packages/Fortress_SharedMaterials/Base_Material_Functions/SwizzleWorldAlignedSpecOutputForFortnite
graph {
 properties {
  description=World sligned texture complex assumes a different texture format than fortnite for spec textures
  libraryCategoriesText [Misc,Fortnite]
 }
 nodes [
  n1 = BreakMaterialAttributes!(Attr=FunctionInput(Preview=MakeMaterialAttributes!(
     BaseColor
     Metallic
     Specular
     Roughness
     Anisotropy
     EmissiveColor
     Opacity
     OpacityMask
     Normal
     Tangent
     WorldPositionOffset
     SubsurfaceColor
     ClearCoat
     ClearCoatRoughness
     AmbientOcclusion
     Refraction
     CustomizedUVs_0
     CustomizedUVs_1
     CustomizedUVs_2
     CustomizedUVs_3
     CustomizedUVs_4
     CustomizedUVs_5
     CustomizedUVs_6
     CustomizedUVs_7
     PixelDepthOffset
     ShadingModel
     Displacement
    )) {bUsePreviewValueAsDefault=true, inputType=FunctionInput_MaterialAttributes})
  FunctionOutput(None=MakeMaterialAttributes(
    BaseColor=@n1.BaseColor
    Metallic=@n1.BaseColor
    Specular=@n1.BaseColor
    Roughness=@n1.BaseColor
    EmissiveColor=@n1.BaseColor
    Opacity=@n1.BaseColor
    OpacityMask=@n1.BaseColor
    Normal=@n1.BaseColor
    WorldPositionOffset=@n1.BaseColor
    SubsurfaceColor=@n1.BaseColor
    ClearCoat=@n1.BaseColor
    ClearCoatRoughness=@n1.BaseColor
    AmbientOcclusion=@n1.BaseColor
    Refraction=@n1.BaseColor
    CustomizedUVs_0=@n1.BaseColor
    CustomizedUVs_1=@n1.BaseColor
    CustomizedUVs_2=@n1.BaseColor
    CustomizedUVs_3=@n1.BaseColor
    CustomizedUVs_4=@n1.BaseColor
    CustomizedUVs_5=@n1.BaseColor
    CustomizedUVs_6=@n1.BaseColor
    CustomizedUVs_7=@n1.BaseColor
    PixelDepthOffset=@n1.BaseColor
   ))
 ]
}
