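# Material function that builds a concrete layer from street-road diffuse, spec mask and normal textures, with an alpha for blending it by vertex color.
# A static switch picks world-aligned projection by default or mesh UVs with optional scale and offset.
# The layer's opacity serves as height for two HeightLerps driven by red vertex color.
# Their remapped, opposing results multiply into a black tint mask along the transition, and the first is also output as the lerp alpha.
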MaterialFunction /FigmentEnvironment/FN_6_20/Packages/DS_Fortnight/SM/Materials/concreteMF
graph {nodes [
  n1 = Constant() {desc=FalloffContrast, r=1}
  n2 = TextureObjectParameter() {
   group=Concrete
   parameterName=Concrete
   samplerSource=SSM_Wrap_WorldGroupSettings
   texture=/FigmentEnvironment/FN_6_20/Packages/DS_Fortnight/SM/Textures/T_Street_Road_D
  }
  n3 = TextureObjectParameter() {
   group=Concrete
   parameterName=Concrete_SpecMask
   samplerSource=SSM_Wrap_WorldGroupSettings
   samplerType=SAMPLERTYPE_Masks
   texture=/FigmentEnvironment/FN_6_20/Packages/DS_Fortnight/SM/Textures/T_Street_Road_M
  }
  n4 = TextureObjectParameter() {
   group=Concrete
   parameterName=ConcreteTextureNormal
   samplerSource=SSM_Wrap_WorldGroupSettings
   samplerType=SAMPLERTYPE_Normal
   texture=/FigmentEnvironment/FN_6_20/Packages/DS_Fortnight/SM/Textures/T_Street_Road_N
  }
  n5 = MaterialFunctionCall!(In=MaterialFunctionCall!(
    Diffuse Texture Object=FunctionInput(Preview=@n2) {bUsePreviewValueAsDefault=true, inputName=Concrete diffuse, inputType=FunctionInput_Texture2D}
    Spec Texture Object=FunctionInput(Preview=@n3) {bUsePreviewValueAsDefault=true, inputName=Concrete Spec, inputType=FunctionInput_Texture2D, sortPriority=1}
    TextureSize=ScalarParameter() {defaultValue=768, group=Concrete, parameterName=ConcTexSize}
    --------------------------
    Use seperate settings for Z projection
    Use seperate texture for Z projection
    Texture Size Z projection
    Z Plane Diffuse Texture Object
    Z Plane Spec Texture Object
    "--------------------------      "
    ProjectionTransitionContrast
    Use Texture Alpha Heightmap to Refine the Transistion
    HeightLerpTransitionContrast
    PickTheDiffuseHeightChannel
    TheDiffuseHeightChannel
    "--------------------------       "
    NormalTextureObject=FunctionInput(Preview=@n4) {bUsePreviewValueAsDefault=true, inputName=Concrete Normal, inputType=FunctionInput_Texture2D, sortPriority=2}
    World Space Normals
    "--------------------------  "
    Export Float 4 Diffuse=StaticBool() {value=true}
    Export Float 4 Spec
    "--------------------------     "
    WorldPosition
    World Space Normal
   ) {materialFunction=/Engine/Functions/Engine_MaterialFunctions02/Texturing/WorldAlignedTextures_Complex}.Z Mat Attributes - See Tooltip) {materialFunction=/FigmentEnvironment/FN_6_20/Packages/Fortress_SharedMaterials/Base_Material_Functions/SwizzleWorldAlignedSpecOutputForFortnite}
  n6 = TextureCoordinate()
  n7 = VectorParameter() {group=Concrete, parameterName=ConcreteTextureCoordinateControls_RG_Scale_BA_Offset, defaultValue=<rgba 1 1 0 0>}
  n8 = ComponentMask(None=@n7.RGB) {g=true, r=true}
  n9 = StaticSwitchParameter(True=Add(A=Multiply(A=@n6, B=@n8), B=AppendVector(A=@n7.B, B=@n7.B)), False=@n6) {group=Concrete, parameterName=Modify Concrete UVS}
  n10 = TextureSampleParameter2D(UVs=@n9) {
   group=Concrete
   parameterName=Concrete
   samplerSource=SSM_Wrap_WorldGroupSettings
   texture=&T_Street_Road_D
  }
  n11 = TextureSampleParameter2D(UVs=@n9) {
   group=Concrete
   parameterName=Concrete_SpecMask
   samplerSource=SSM_Wrap_WorldGroupSettings
   samplerType=SAMPLERTYPE_Masks
   texture=&T_Street_Road_M
  }
  n12 = TextureSampleParameter2D(UVs=@n9) {
   group=Concrete
   parameterName=ConcreteTextureNormal
   samplerSource=SSM_Wrap_WorldGroupSettings
   samplerType=SAMPLERTYPE_Normal
   texture=&T_Street_Road_N
  }
  n13 = Transform(None=@n12.RGB)
  n14 = VertexNormalWS()
  n15 = StaticBoolParameter() {defaultValue=true, group=Concrete, parameterName=Use Concrete Normals}
  n16 = StaticSwitchParameter(
   True=@n5.Result
   False=MakeMaterialAttributes!(
    BaseColor=@n10.RGB
    Metallic=@n11.G
    Specular=@n11.G
    Roughness=@n11.G
    Anisotropy
    EmissiveColor
    Opacity=@n10.RGB
    OpacityMask
    Normal=StaticSwitch(True=@n13, False=@n14, Value=@n15)
    Tangent
    WorldPositionOffset
    SubsurfaceColor
    ClearCoat
    ClearCoatRoughness
    AmbientOcclusion
    Refraction
    CustomizedUVs_0
    CustomizedUVs_1
    CustomizedUVs_2
    CustomizedUVs_3
    CustomizedUVs_4
    CustomizedUVs_5
    CustomizedUVs_6
    CustomizedUVs_7
    PixelDepthOffset
    ShadingModel
    Displacement
   )
  ) {defaultValue=true, group=Concrete, parameterName=UseWorldPositionCoordinatesForConcrete}
  n17 = MaterialFunctionCall!(Mat attributes=@n16) {materialFunction=/Engine/Functions/Engine_MaterialFunctions02/MatLayerBlend_BreakOpacity}
  n18 = VertexColor()
  n19 = Clamp(None=LinearInterpolate(
    A=Constant() {desc=FalloffPushLower, r=-0.5}
    B=Constant() {desc=FalloffPushUpper, r=3}
    Alpha=MaterialFunctionCall!(Lerp A, Lerp B, Height Texture=@n17.Result, Falloff Contrast=@n1, Transition Phase=@n18.R) {materialFunction=/FigmentEnvironment/FN_6_20/Packages/Fortnight_MaterialFunctions/Temp/HeightLerp}.Lerp Alpha
   ))
  n20 = Clamp(None=LinearInterpolate(
    A=Constant() {desc=FalloffPushUpper, r=1}
    B=Constant() {desc=FalloffPushLower, r=-0.5}
    Alpha=MaterialFunctionCall(Height Texture=@n17.Result, Transition Phase=@n18.R) {materialFunction=&HeightLerp}.Lerp Alpha No Contrast
   ))
  n21 = MaterialFunctionCall!(
   Mat attributes=@n16
   Use Spec=StaticBoolParameter() {defaultValue=true, group=Concrete, parameterName=Use Concrete Spec}
   Use Normals=StaticBoolParameter() {defaultValue=true, group=Concrete, parameterName=Use Concrete Normals}
  ) {materialFunction=/FigmentEnvironment/FN_6_20/Packages/Fortnite_UniqueMaterials_1/MaterialFunctions/Fortnite_Terrain_invertOpacityNormalBool}
  FunctionOutput(None=MaterialFunctionCall!(Material=@n21.Result, Tint=Constant(), TintMask=Multiply(A=@n19, B=@n20)) {materialFunction=/Engine/Functions/MaterialLayerFunctions/MatLayerBlend_Tint}.Blended Material) {outputName=Material Attribute}
  FunctionOutput(None=@n19) {outputName="Lerp Alpha (controlled by the red vertex color)", sortPriority=1}
 ]}
